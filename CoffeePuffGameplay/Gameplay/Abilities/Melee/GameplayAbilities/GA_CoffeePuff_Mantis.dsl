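; Gameplay ability for the CoffeePuff Mantis melee weapon, overriding the level used when applying its gameplay effect containers.
; GetLevelForGameplayEffectContainer reads the current source weapon's data tags and returns level 1 for Rarity.Mythic, otherwise 0.
; An invalid weapon falls through the sequence to return 0, and the ApplicationTag input is unused.
; This lets the Mythic variant deal damage at a different effect level.
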(fn GetLevelForGameplayEffectContainer
  (params (ApplicationTag :type Gameplay%20Tag%20Structure))
  (bind _returnvalue (Ability|GetCurrentSourceWeapon))
  ;[0.15 0.03959 0.1303 0.5] GetLevelForGameplayEffectContainer: Mythic variant deals damage at a different level.
  (Utilities|FlowControl|Sequence
    (:then_0
      (Utilities|IsValid _returnvalue
        (:Is%20Valid
          (return (Math|Integer|SelectInt 1 0 (GameplayTags|HasTag (Tags|GetWeaponDataGameplayTags _returnvalue) "(TagName=\"Rarity.Mythic\")"))))))
    (:then_1
      (return 0))))
